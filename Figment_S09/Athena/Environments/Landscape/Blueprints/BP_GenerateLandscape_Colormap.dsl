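; Editor tool Blueprint that bakes a landscape colormap through a SceneCapture2D into a target Texture2D.
; Its construction script pushes display toggles and colormap tuning values into the MPC_Landscape parameter collection.
; Separate custom events assign the render material to landscape and streaming proxies, trigger the capture, and convert the render target to texture.
; Another event rebuilds the capture's show-only list from landscape spline components alone, printing debug strings.
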(event Custom|Restore%20Landscape%20Mats
  (bind _outactors (Actor|GetAllActorsOfClass "/Script/Landscape.LandscapeStreamingProxy"))
  (for _array_element _outactors
    (Variables|/Script/Landscape.LandscapeProxy|SetLandscapeMaterial (Variables|Default|GetRender%20Material) _array_element))
  (bind _outactors_2 (Actor|GetAllActorsOfClass "/Script/Landscape.Landscape"))
  (Variables|/Script/Landscape.LandscapeProxy|SetLandscapeMaterial (Variables|Default|GetRender%20Material) (Utilities|Array|Get(acopy) _outactors_2)))

(event Custom|Trigger%20Scene%20Capture
  (Rendering|SceneCapture|CaptureScene (Variables|/Script/Engine.SceneCapture2D|GetCaptureComponent2D (Variables|Default|GetScene%20Capture%20Actor)) :!class /Script/Engine.SceneCaptureComponent2D))

(event Custom|Overwrite%20Texture%20Target
  (EditorScripting|Rendering|ConvertRenderTargettoTexture2DEditorOnly (Variables|Default|GetRender%20Target) (Variables|Default|GetTarget%20Texture2D)))

(event Custom|Show%20Only%20LS%20and%20Splines
  (Rendering|SceneCapture|ClearShowOnlyComponents (Variables|/Script/Engine.SceneCapture2D|GetCaptureComponent2D (Variables|Default|GetScene%20Capture%20Actor)))
  (bind _outactors (Actor|GetAllActorsOfClass "/Script/Landscape.LandscapeProxy"))
  (for _array_element _outactors
    (Development|PrintString (Utilities|String|Append "Num LS Comps: " (Utilities|String|ToString(Integer) (Utilities|Array|Length (Actor|GetComponentsByClass _array_element "/Script/Landscape.LandscapeSplinesComponent")))))
    (for _array_element_2 (Actor|GetComponentsByClass _array_element "/Script/Landscape.LandscapeSplinesComponent")
      (bind _aslandscape_splines_component (Utilities|Casting|CastToLandscapeSplinesComponent _array_element_2))
      (Rendering|SceneCapture|ShowOnlyComponent (Variables|/Script/Engine.SceneCapture2D|GetCaptureComponent2D (Variables|Default|GetScene%20Capture%20Actor)) _aslandscape_splines_component)
      (Development|PrintString "cast wiorked"))))

(fn ConstructionScript
  (Rendering|Material|SetScalarParameterValue "/Game/Athena/Environments/Landscape/MPC/MPC_Landscape.MPC_Landscape" "DisplayLowDetailShader" (Math|Conversions|ToFloat(Boolean) (Variables|Default|GetDisplay%20Low%20Detail%20Shader)) :!class /Script/Engine.KismetMaterialLibrary)
  (Rendering|Material|SetScalarParameterValue "/Game/Athena/Environments/Landscape/MPC/MPC_Landscape.MPC_Landscape" "RenderColorMap" (Math|Conversions|ToFloat(Boolean) (Variables|Default|GetDisplay%20Color%20Map)) :!class /Script/Engine.KismetMaterialLibrary)
  (Rendering|Material|SetScalarParameterValue "/Game/Athena/Environments/Landscape/MPC/MPC_Landscape.MPC_Landscape" "DisplayBiomeRGB" (Math|Conversions|ToFloat(Boolean) (Variables|Default|GetDisplay%20Biome%20RGB)) :!class /Script/Engine.KismetMaterialLibrary)
  (Rendering|Material|SetScalarParameterValue "/Game/Athena/Environments/Landscape/MPC/MPC_Landscape.MPC_Landscape" "ColormapDivisor" (Variables|Default|GetDivisor) :!class /Script/Engine.KismetMaterialLibrary)
  (Rendering|Material|SetScalarParameterValue "/Game/Athena/Environments/Landscape/MPC/MPC_Landscape.MPC_Landscape" "ColormapTransitionBias" (Variables|Default|GetGrass%20Cliff%20Bias) :!class /Script/Engine.KismetMaterialLibrary)
  (Rendering|Material|SetScalarParameterValue "/Game/Athena/Environments/Landscape/MPC/MPC_Landscape.MPC_Landscape" "GrassEdgeZ" (Variables|Default|GetGrass%20Edge%20Z) :!class /Script/Engine.KismetMaterialLibrary)
  (Rendering|Material|SetScalarParameterValue "/Game/Athena/Environments/Landscape/MPC/MPC_Landscape.MPC_Landscape" "GrassEdgeDesatWidth" (Variables|Default|GetGrass%20Edge%20Desat%20Width) :!class /Script/Engine.KismetMaterialLibrary))
